(event EventBeginPlay
  (bind (_gamestate _playlist) (Class|FortAsyncActionCurrentPlaylistReady|PlaylistDataReadyAsync)
    (:OnReady
      (CallFunction|SetBossPointProvider)
      (CallFunction|Init)
      (Variables|Default|SetWeakpointMaxHealth (ScalableFloat|GetValueatLevel (Variables|Hotfix|GetWeakpointPlaylistMaxHealth_Hf) (Math|Conversions|ToFloat(Integer) (Math|Integer|Clamp(Integer) (Variables|/Script/FortniteGame.FortPlaylist|GetMaxTeamSize _playlist) 1 4)))))))

(event EventAISpawned
  (params (SpawnedActor :type Actor%20Object%20Reference) (bResult :type Boolean))
  (Class|MaskMakerPawnInterface|SetChallengeManagerRef(Message) SpawnedActor self)
  (|CallParentFunction SpawnedActor bResult))

(event EventWeakpointHealthUpdated
  (params (WeakpointIndex :type Integer) (WeakpointHealth :type Float%20%28double-precision%29))
  (Networking|SwitchHasAuthority
    (:Authority
      (Variables|Default|SetWeakpointUpdateRequestId (+ (Variables|Default|GetWeakpointUpdateRequestId) 1)))))

(event EventBossStunned
  (Networking|CanExecuteCosmeticEvents
    (:True
      (Default|CallOnBossMaskPhaseUpdated false))))

(event EventBossRecoveredFromStun
  (Networking|CanExecuteCosmeticEvents
    (:True
      (Default|CallOnBossMaskPhaseUpdated true))))

(fn OnRep_WeakpointUpdateRequestId
  (Networking|CanExecuteCosmeticEvents
    (:True
      (bind _outactors (Actor|GetAllActorswithInterface "/MaskMakerBoss/Interface/MaskMakerWeakpoint_Interface.MaskMakerWeakpoint_Interface_C"))
      (for (_array_element _array_index) _outactors
        (bind _health (Class|MaskMakerWeakpointInterface|GetWeakpointCurrentHealth(Message) _array_element))
        (Default|CallOnWeakpointsHealthUpdated _array_index (Math|Float|SafeDivide _health (Variables|Default|GetWeakpointMaxHealth)))))))

(fn SetBossPointProvider
  (if (ScalableFloat|GetValueasBool (Variables|Internal|GetSF_MMB_Enabled))
    (LocalVariables|SetSelectedPointProvider (Variables|Tunables|GetPointProvider_MaskMaker))
    (elif (ScalableFloat|GetValueasBool (Variables|Internal|GetSF_MMBN_Enabled))
      (LocalVariables|SetSelectedPointProvider (Variables|Tunables|GetPointProvider_MaskMakerNeutral))))
  (Variables|References|SetPatrolPointProvider (LocalVariables|GetSelectedPointProvider)))

(fn ConstructionScript
  (|CallParentFunction))
