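# Translucent UI material for the MaskMaker boss HUD's magic graphic, built from animated noise and flow effects.
# Panning blurred noise at three scales slightly distorts UVs of the magic texture, whose R and G channels blend by B.
# Separate branches blend two fill colors along a linear gradient and threshold flow-map output times texture alpha into a stepped mask.
# Another branch fades distorted alpha with a rotated reflected gradient.
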Material /MaskMakerBoss/UI/Material/M_UI_MaskMaker_Magic
graph {
 properties {bAutomaticallySetUsageInEditor=false, blendMode=BLEND_Translucent, materialDomain=MD_UI, usedShadingModels=Default Lit}
 nodes [
  n1 = TextureCoordinate()
  n2 = Constant() {r=0.1}
  n3 = TextureSample(UVs=Panner(Coordinate=MaterialFunctionCall(UVs=@n1, Texture Scale=@n2) {materialFunction=&ScaleUVsByCenter}.UVs) {speedX=0.15}) {samplerType=SAMPLERTYPE_Grayscale, texture=/MaskMakerBoss/UI/Textures/LowResBlurredNoise}
  n4 = TextureSample(UVs=Add(A=TextureCoordinate(), B=Multiply(A=@n3.R) {constB=0.01})) {texture=/MaskMakerBoss/UI/Textures/T_UI_HUD_MaskMakerMagic}
  n5 = TextureCoordinate()
  n6 = MaterialFunctionCall(UVs=TextureCoordinate(), Texture Scale=Constant() {r=0.5}) {materialFunction=&ScaleUVsByCenter}
  n7 = Multiply(A=TextureSample(UVs=Panner(Coordinate=@n6.UVs) {speedX=0.15}) {samplerType=SAMPLERTYPE_Grayscale, texture=&LowResBlurredNoise}.R) {constB=0.01}
  n8 = TextureCoordinate()
  n9 = MaterialFunctionCall(UVs=TextureCoordinate(), Texture Scale=Constant() {r=1}) {materialFunction=&ScaleUVsByCenter}
  n10 = Multiply(A=TextureSample(UVs=Panner(Coordinate=@n9.UVs) {speedX=0.15}) {samplerType=SAMPLERTYPE_Grayscale, texture=&LowResBlurredNoise}.R) {constB=0.01}
  LinearInterpolate(
   A=@n4.R
   B=TextureSample(UVs=Add(A=@n5, B=@n7)) {texture=&T_UI_HUD_MaskMakerMagic}.G
   Alpha=TextureSample(UVs=Add(A=@n8, B=@n10)) {texture=&T_UI_HUD_MaskMakerMagic}.B
  )
  n11 = VectorParameter() {parameterName=Fill Color 1, defaultValue=<rgba 0.0513 0.0284 0.0086 1>}
  n12 = VectorParameter() {parameterName=Fill Color 2, defaultValue=<rgba 0.8632 0.7529 0.3712 1>}
  n13 = MaterialFunctionCall(UVs=TextureCoordinate(), Texture=TextureSample() {texture=/MaskMakerBoss/UI/Textures/T_UI_HUD_MaskMakerMagicFlow}.RGB) {materialFunction=&MF_UI_LinearGradient}
  LinearInterpolate(A=@n11.RGB, B=@n12.RGB, Alpha=@n13.Result)
  n14 = MaterialFunctionCall!(
   Texture 2D=TextureObject() {texture=&T_UI_HUD_MaskMakerMagicFlow}
   Texture UVs
   FlowMap Texture 2D=TextureObject() {texture=&T_UI_HUD_MaskMakerMagicFlow}
   FlowMap UVs
   FlowIntensity
   FlowSpeed=Constant() {r=0.2}
  ) {materialFunction=/FortUILibrary/MaterialFunctions/Effects/MF_UI_FlowMap}
  n15 = TextureSample() {texture=&T_UI_HUD_MaskMakerMagic}
  Multiply(A=LinearInterpolate(A=@n11.A, B=@n12.A, Alpha=@n13.Result), B=Step(X=Multiply(A=@n14.Result, B=@n15.A)) {constX=0.1, constY=0.017})
  n16 = TextureCoordinate()
  n17 = Constant() {r=0.25}
  Multiply(
   A=@n4.A
   B=Saturate(None=MaterialFunctionCall!(UVs=@n16, Offset, RotationCenter, RotationAngle=@n17) {materialFunction=/FortUILibrary/MaterialFunctions/Gradients/MF_UI_ReflectedGradient}.Result)
  )
 ]
}
