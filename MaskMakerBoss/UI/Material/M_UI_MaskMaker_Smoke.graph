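# Translucent UI material for the MaskMaker boss that builds a drifting smoke pattern.
# Scaled-to-fill UVs pass through a warp, then drive two counter-panning cloud and noise samples that are weighted and summed.
# That sum joins a lerp of scrolling and static cloud textures, then is multiplied by an eased radial and linear gradient mask.
# An Atmospherics Color parameter exists but feeds nothing in the graph.
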Material /MaskMakerBoss/UI/Material/M_UI_MaskMaker_Smoke
graph {
 properties {bAutomaticallySetUsageInEditor=false, blendMode=BLEND_Translucent, materialDomain=MD_UI, usedShadingModels=Default Lit}
 nodes [
  n1 = TextureSample(UVs=Panner(Time=Time()) {speedY=0.03}) {texture=/MaskMakerBoss/UI/Textures/clouds_dust_detail_01}
  n2 = TextureSample() {texture=&clouds_dust_detail_01}
  n3 = MaterialFunctionCall!(
   UV=MaterialFunctionCall(UV=TextureCoordinate()) {materialFunction=&MF_UI_UVScaling}.Scaled To Fill
   Warp Center=AppendVector(A=Constant() {r=0.5}, B=Constant() {r=0.75})
   Warp Intensity=Constant() {r=0.25}
   Warp Radius=Constant() {r=0.35}
  ) {materialFunction=/FortUILibrary/MaterialFunctions/Effects/MF_UI_Warp}
  n4 = Time()
  n5 = Panner(Coordinate=Multiply(A=@n3.Warped UV) {constB=3}, Time=@n4) {speedX=0.007, speedY=0.02}
  n6 = Panner(Coordinate=Multiply(A=@n3.Warped UV) {constB=2}, Time=@n4) {speedX=-0.008, speedY=-0.02}
  n7 = Add(
   A=Multiply(A=TextureSample(UVs=@n5) {texture=&clouds_dust_detail_01}.RGB) {constB=0.6}
   B=Multiply(A=TextureSample(UVs=@n6) {samplerType=SAMPLERTYPE_Grayscale, texture=/MaskMakerBoss/UI/Textures/LowResBlurredNoise}.RGB) {constB=0.3}
  )
  n8 = TextureCoordinate()
  n9 = AppendVector(A=Constant(), B=Constant())
  n10 = MaterialFunctionCall("Value (0-1)"=MaterialFunctionCall(UV=Add(A=@n8, B=@n9), GradientEnd=Constant() {r=0.5}) {materialFunction=&MF_UI_RadialGradient}.Smooth) {materialFunction=&MF_UI_EaseCurves}
  n11 = MaterialFunctionCall(GradientStart=Constant() {r=0.25}, GradientEnd=Constant() {r=0.45}, RotationAngle=Constant() {r=0.25}) {materialFunction=&MF_UI_LinearGradient}
  Multiply(
   A=Add(A=LinearInterpolate(A=@n1.RGB, Alpha=@n2.RGB), B=Saturate(None=@n7))
   B=Power(Base=Multiply(A=@n10.EaseOutSine, B=@n11.Result)) {constExponent=0.5}
  )
  VectorParameter() {parameterName=Atmospherics Color, defaultValue=<rgba 1 1 1 1>}
 ]
}
