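# UI-domain translucent material for a MaskMakerBoss widget background, building a hexagon with a two-color linear gradient fill.
# Two custom HLSL polygon SDFs form upper and bottom trapezoids, antialiased with screen derivatives and exposed as named reroutes.
# Gradient alpha is multiplied by a sum of two reroutes; the pixel-size scaling divides components by themselves, so applies no aspect correction.
# A separate library trapezoid intersection branch feeds no other node.
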Material /MaskMakerBoss/UI/Material/M_UI_MaskMaker_bg
graph {
 properties {bAutomaticallySetUsageInEditor=false, blendMode=BLEND_Translucent, materialDomain=MD_UI, usedShadingModels=Default Lit}
 nodes [
  n1 = VectorParameter() {parameterName=Fill Color 1, defaultValue=<rgba 1 0.2755 0 1>}
  n2 = VectorParameter() {parameterName=Fill Color 2, defaultValue=<rgba 0 0.8768 1 1>}
  n3 = MaterialFunctionCall(UVs=TextureCoordinate(), GradientStart=Constant() {r=0.1}, GradientEnd=Constant() {r=0.9}) {materialFunction=&MF_UI_LinearGradient}
  LinearInterpolate(A=@n1.RGB, B=@n2.RGB, Alpha=@n3.Result)
  Multiply(A=LinearInterpolate(A=@n1.A, B=@n2.A, Alpha=@n3.Result), B=Add(A=NamedRerouteUsage(), B=NamedRerouteUsage()))
  n4 = MaterialFunctionCall(Float2=MaterialFunctionCall() {materialFunction=&GetUserInterfaceUV}.Pixel Size) {materialFunction=&BreakOutFloat2Components}
  n5 = Constant() {r=1}
  n6 = Max(A=AppendVector(A=Divide(A=@n4.R, B=@n4.R), B=@n5) {desc=Horizontal}, B=AppendVector(A=@n5, B=Divide(A=@n4.G, B=@n4.G)) {desc=Vertical})
  MaterialExpressionMultiply_10 = Multiply(A=@n6, B=TextureCoordinate())
  MaterialExpressionMultiply_11 = Multiply(A=@n6, B=Constant2Vector() {r=0.15})
  MaterialExpressionMultiply_12 = Multiply(A=@n6, B=Constant2Vector() {r=0.85})
  MaterialExpressionMultiply_13 = Multiply(A=@n6, B=Constant2Vector() {g=0.5, r=1})
  MaterialExpressionMultiply_14 = Multiply(A=@n6, B=Constant2Vector() {g=0.5})
  n7 = Custom!(
   p=@MaterialExpressionMultiply_10
   v1=@MaterialExpressionMultiply_11
   v2=@MaterialExpressionMultiply_12
   v3=@MaterialExpressionMultiply_13
   v4=@MaterialExpressionMultiply_14
  ) {
   code="float2 v[] = {v1,v2,v3,v4};\r\nconst int N = 4;\r\n\r\nfloat d = dot(p-v[0],p-v[0]);\r\nfloat s = 2.0;\r\nfor( int i=0, j=N-1; i<N; j=i, i++ )\r\n{\r\n    float2 e = v[j] - v[i];\r\n    float2 w = p - v[i];\r\n    float2 b = w - e*clamp( dot(w,e)/dot(e,e), 0.0, 1.0 );\r\n    d = min( d, dot(b,b) );\r\n    float3 c = float3(p.y>=v[i].y,p.y<v[j].y,e.x*w.y>e.y*w.x);\r\n    if( all(c) || all(!c) ) s = -s;\r\n}\r\nreturn s*sqrt(d);"
   description=PolySDF
   outputType=CMOT_Float1
   inputs [
    {inputName=p, input {expression=&self:MaterialExpressionMultiply_10}}
    {inputName=v1, input {expression=&self:MaterialExpressionMultiply_11}}
    {inputName=v2, input {expression=&self:MaterialExpressionMultiply_12}}
    {inputName=v3, input {expression=&self:MaterialExpressionMultiply_13}}
    {inputName=v4, input {expression=&self:MaterialExpressionMultiply_14}}
   ]
  }
  n8 = AppendVector(A=DDX(Value=@n7), B=DDY(Value=@n7))
  n9 = Divide(A=@n7, B=Add(A=Length(None=@n8) {is=class MaterialExpressionLength}) {constB=0})
  NamedRerouteDeclaration(None=OneMinus(None=Saturate(None=@n9))) {name=Upper Shape, nodeColor=<rgba 0 0 0 1>}
  n10 = MaterialFunctionCall(Float2=MaterialFunctionCall() {materialFunction=&GetUserInterfaceUV}.Pixel Size) {materialFunction=&BreakOutFloat2Components}
  n11 = Constant() {r=1}
  n12 = Max(A=AppendVector(A=Divide(A=@n10.R, B=@n10.R), B=@n11) {desc=Horizontal}, B=AppendVector(A=@n11, B=Divide(A=@n10.G, B=@n10.G)) {desc=Vertical})
  MaterialExpressionMultiply_6 = Multiply(A=@n12, B=TextureCoordinate())
  MaterialExpressionMultiply_7 = Multiply(A=@n12, B=Constant2Vector() {g=0.5})
  MaterialExpressionMultiply_8 = Multiply(A=@n12, B=Constant2Vector() {g=0.5, r=1})
  MaterialExpressionMultiply_9 = Multiply(A=@n12, B=Constant2Vector() {g=1, r=0.85})
  MaterialExpressionMultiply_15 = Multiply(A=@n12, B=Constant2Vector() {g=1, r=0.15})
  n13 = Custom(
   p=@MaterialExpressionMultiply_6
   v1=@MaterialExpressionMultiply_7
   v2=@MaterialExpressionMultiply_8
   v3=@MaterialExpressionMultiply_9
   v4=@MaterialExpressionMultiply_15
  ) {
   code="float2 v[] = {v1,v2,v3,v4};\r\nconst int N = 4;\r\n\r\nfloat d = dot(p-v[0],p-v[0]);\r\nfloat s = 2.0;\r\nfor( int i=0, j=N-1; i<N; j=i, i++ )\r\n{\r\n    float2 e = v[j] - v[i];\r\n    float2 w = p - v[i];\r\n    float2 b = w - e*clamp( dot(w,e)/dot(e,e), 0.0, 1.0 );\r\n    d = min( d, dot(b,b) );\r\n    float3 c = float3(p.y>=v[i].y,p.y<v[j].y,e.x*w.y>e.y*w.x);\r\n    if( all(c) || all(!c) ) s = -s;\r\n}\r\nreturn s*sqrt(d);"
   description=PolySDF
   outputType=CMOT_Float1
   inputs [
    {inputName=p, input {expression=&self:MaterialExpressionMultiply_6}}
    {inputName=v1, input {expression=&self:MaterialExpressionMultiply_7}}
    {inputName=v2, input {expression=&self:MaterialExpressionMultiply_8}}
    {inputName=v3, input {expression=&self:MaterialExpressionMultiply_9}}
    {inputName=v4, input {expression=&self:MaterialExpressionMultiply_15}}
   ]
  }
  n14 = AppendVector(A=DDX(Value=@n13), B=DDY(Value=@n13))
  n15 = Divide(A=@n13, B=Add(A=Length(None=@n14)) {constB=0})
  NamedRerouteDeclaration(None=OneMinus(None=Saturate(None=@n15))) {name=BottomShape, nodeColor=<rgba 0 0 0 1>}
  n16 = Constant() {r=1}
  n17 = Constant() {r=0.8}
  n18 = Constant() {r=1}
  n19 = MaterialFunctionCall() {materialFunction=&GetUserInterfaceUV}
  n20 = MaterialFunctionCall!(UVs=@n19.Normalized UV, TopWidth=@n16, BottomWidth=@n17, Height=@n18, OutlineThickness, OutlineAlignment, GlowMin, GlowMax) {materialFunction=/FortUILibrary/MaterialFunctions/Shapes/MF_UI_IsoscelesTrapezoid}
  n21 = MaterialFunctionCall(UVs=@n19.Normalized UV, TopWidth=@n17, BottomWidth=@n18, Height=@n16) {materialFunction=&MF_UI_IsoscelesTrapezoid}
  OneMinus(None=Step(X=MaterialFunctionCall!(DistanceFieldA=@n20.SDF, DistanceFieldB=@n21.SDF, BlendAmount) {materialFunction=/FortUILibrary/MaterialFunctions/Shapes/Modifiers/MF_UI_SDF_Intersection}.Exact))
 ]
}
