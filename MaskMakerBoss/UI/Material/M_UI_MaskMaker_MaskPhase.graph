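# Translucent UI material for the MaskMaker boss, named for its mask phase, that draws a gold double-bordered diamond.
# UI UVs pass through a scaling function and an eighth-turn rotation, then feed two concentric SDF boxes.
# Outlines of both boxes are tinted gold and added to the outer box's black fill, which contributes only to the alpha channel.
# All values are constants, with no exposed parameters or animation.
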Material /MaskMakerBoss/UI/Material/M_UI_MaskMaker_MaskPhase
graph {
 properties {bAutomaticallySetUsageInEditor=false, blendMode=BLEND_Translucent, materialDomain=MD_UI, usedShadingModels=Default Lit}
 nodes [
  n1 = Constant4Vector() {constant=<rgba 0.8714 0.6724 0.1779 1>}
  n2 = MaterialFunctionCall(
   UVs=MaterialFunctionCall(UV=MaterialFunctionCall() {materialFunction=&GetUserInterfaceUV}.Normalized UV) {materialFunction=&MF_UI_UVScaling}.ScaleY
   "Rotation Angle (0-1)"=Constant() {r=0.125}
  ) {materialFunction=&CustomRotator}
  n3 = MaterialFunctionCall(UVs=@n2.Rotated Values, Size=Constant() {r=0.66}) {materialFunction=&MF_UI_Box}
  n4 = MaterialFunctionCall!(DistanceField=@n3.SDF, Threshold, Thickness, Alignment, Smoothing) {materialFunction=/FortUILibrary/MaterialFunctions/Shapes/Modifiers/MF_UI_ShapeOutline}
  n5 = MaterialFunctionCall(
   DistanceField=MaterialFunctionCall(UVs=@n2.Rotated Values, Size=Constant() {r=0.5}) {materialFunction=&MF_UI_Box}.SDF
   Thickness=Constant() {r=0.02}
  ) {materialFunction=&MF_UI_ShapeOutline}
  Add(
   A=Add(A=LinearInterpolate(B=@n1.RGBA, Alpha=@n4.Outline), B=LinearInterpolate(B=@n1.RGBA, Alpha=@n5.Outline))
   B=LinearInterpolate(B=Constant4Vector() {constant=<rgba 0 0 0 1>}.RGBA, Alpha=@n3.Fill)
  )
 ]
}
