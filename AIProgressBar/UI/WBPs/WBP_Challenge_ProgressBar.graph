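# WBP_Challenge_ProgressBar is a UserWidget interface for a challenge progress bar with minion and boss stages, though most of its logic is empty.
# Functions and events declare direct or animated progress, shake, boss icon and style switching, alongside two-color palettes per stage.
# Construct calls the empty SetWidthMobile, and a custom event broadcasts the fill-finished dispatcher.
# The widget tree is empty and GetSizeBoxWidth returns zero.
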WidgetBlueprint /AIProgressBar/UI/WBPs/WBP_Challenge_ProgressBar
graphs [
 EventGraph {nodes [
   CallBarAnimationFinishedEvent() {is=title K2Node_CustomEvent}
   > Call On Bar Fill Animation Finished!(execute:Exec, self:WBP Challenge Progress Bar Object Reference) -> (then:Exec) {is=title K2Node_CallDelegate}
   Event Construct()
   > SetWidthMobile()
   PlayShakeAnimation!() -> (OutputDelegate, then, AnimationDelay:double=~) {is=title K2Node_CustomEvent}
   SetProgressBarStyle!() -> (OutputDelegate, then, HasReachedBossStage:bool=~) {is=title K2Node_CustomEvent}
   WaitForBarAnimationToFinish() {is=title K2Node_CustomEvent}
  ]}
 GetSizeBoxWidth {nodes [
   GetSizeBoxWidth() {is=title K2Node_FunctionEntry}
   > Return Node!(execute, Width:double=0)
  ]}
 OnBarFillAnimationFinished {nodes [OnBarFillAnimationFinished() {is=title K2Node_FunctionEntry}]}
 Sequencer Events {links [], nodes []}
 SetProgressDirectly {nodes [SetProgressDirectly!() -> (then, Progress:double=~) {is=title K2Node_FunctionEntry}]}
 SetProgressWithAnimation {nodes [SetProgressWithAnimation!() -> (then, ProgressEnd:double=~) {is=title K2Node_FunctionEntry}]}
 SetWidthMobile {nodes [SetWidthMobile() {is=title K2Node_FunctionEntry}]}
 SetupBossBarElements {nodes [SetupBossBarElements!() -> (then, IsInBossStage:bool=~) {is=title K2Node_FunctionEntry}]}
 SetupBossIcon {nodes [SetupBossIcon!() -> (then, BossIcon:Texture Object Reference) {is=title K2Node_FunctionEntry}]}
]
variables [
 BarAnimTimer {type=TimerHandle}
 LerpAlpha {type=number}
 OnBarFillAnimationFinished {}
 ProgressBarStyle_Boss {type="array<LinearColor>", default [<rgba 0.8469 0.0116 0.0176 1>, <rgba 0.6654 0.1329 0 1>]}
 ProgressBarStyle_Minions {type="array<LinearColor>", default [<rgba 0.807 0.3864 0 1>, <rgba 1 0.7835 0 1>]}
]
widgets {info {inheritedWidgetCount=0, parentClass=&UserWidget, rootWidgetClass=null, widgetCount=0}, widgets []}
