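# Mask icon pip widget for the Mask Maker boss challenge UI that shows whether a weakpoint is active or destroyed.
# SetWeakpointActive passes the new state through a Delta Gate macro on Is Active, then plays the animation selected by that boolean.
# The active animation fades the icon from dim to full opacity, and the destroyed animation fades it back to dim.
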WidgetBlueprint /MaskMakerBoss/UI/WBP_MaskMaker_Challenge_MaskPip
animations [
 Anim_WeakpointDestroyed {
  binding_count=1
  length_seconds=0.3334
  object=&self:Anim_WeakpointDestroyed
  sequence {
   bindings [Image_MaskIcon {
     id=#1
     possessed_class=Image
     tracks [{
       type=MovieSceneFloatTrack
       display_name=Render Opacity
       sections [{type=MovieSceneFloatSection, end_frame=20, start_frame=0, channels [None {type=MovieSceneScriptingFloatChannel, keys [[0,1],[20,0.25]]}]}]
      }]
    }]
   display_rate {numerator=60}
  }
 }
 Anim_WeakpointActive {
  binding_count=1
  length_seconds=0.3334
  object=&self:Anim_WeakpointActive
  sequence {
   bindings [Image_MaskIcon {
     id=#2
     possessed_class=Image
     tracks [{
       type=MovieSceneFloatTrack
       display_name=Render Opacity
       sections [{type=MovieSceneFloatSection, end_frame=20, start_frame=0, channels [None {type=MovieSceneScriptingFloatChannel, keys [[0,0.25],[20,1]]}]}]
      }]
    }]
   display_rate {numerator=60}
  }
 }
]
graphs [
 EventGraph {links [], nodes []}
 SetWeakpointActive {nodes [
   n1 = SetWeakpointActive!() -> (then, IsActive:bool=~) {is=title K2Node_FunctionEntry}
   > Delta Gate Wildcard!(execute:Exec, Variable:"Boolean (by ref)"=Get Is Active() -> (Is Active:bool), New Value:bool=@n1) -> (Success:Exec, Updated Variable:"Boolean (by ref)", Previous Variable:bool=~) {is=title K2Node_MacroInstance}
   > PlayAnimation(InAnimation=Select(
     Option 0:Widget Animation Object Reference=Get Anim_WeakpointDestroyed() -> (Anim_WeakpointDestroyed:Widget Animation Object Reference)
     Option 1:Widget Animation Object Reference=Get Anim_WeakpointActive() -> (Anim_WeakpointActive:Widget Animation Object Reference)
     Index=Get Is Active()
    ) -> (ReturnValue:Widget Animation Object Reference))
  ]}
]
variables [Is Active {type=boolean}]
widgets {
 info {inheritedWidgetCount=0, parentClass=&UserWidget, rootWidgetClass=&Overlay, widgetCount=2}
 widgets [
  Overlay_19 {widgetClassPath=&Overlay, widget=Overlay_19}
  Image_MaskIcon {
   widgetClassPath=&Image
   bIsVariable=true
   parent=Overlay_19
   widget=Image_MaskIcon
   properties {
    visibility=SelfHitTestInvisible
    brush {
     bIsDynamicallyLoaded=false
     drawAs=Image
     imageType=FullColor
     mirroring=NoMirror
     resourceName=None
     resourceObject=/MaskMakerBoss/UI/Textures/T_MaskMaker_PhaseIcon
     tiling=NoTile
     imageSize=<32 32>
     margin=<ltrb 0 0 0 0>
     outlineSettings {
      bUseBrushTransparency=false
      roundingType=HalfHeightRadius
      width=0
      color {colorUseRule=UseColor_Specified, specifiedColor=<rgba 0 0 0 0>}
      cornerRadii=<0 0 0 0>
     }
     tintColor {colorUseRule=UseColor_Specified, specifiedColor=<rgba 1 1 1 1>}
     uVRegion {bIsValid=false, max=<0 0>, min=<0 0>}
    }
   }
   slot {class=OverlaySlot, properties {horizontalAlignment=HAlign_Center, verticalAlignment=VAlign_Center}}
  }
 ]
}
