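# Mask Maker boss variant of the challenge progress bar widget, adding a mask-phase row of four weakpoint pips.
# Progress drives the fill material's Progress parameter, either directly or through a looping timer that steps down to the target.
# Mask state changes play intro/outro animations; active weakpoint indices light pips by count rather than by index.
# It also plays boss-bar switch and delayed shake animations; SetupBossBarElements and SetupBossIcon are empty.
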WidgetBlueprint /MaskMakerBoss/UI/WBP_MaskMaker_Challenge_ProgressBar
animations [
 Anim_SwitchToBossBar {
  binding_count=1
  length_seconds=0.8
  object=&self:Anim_SwitchToBossBar
  sequence {bindings [Overlay_Root {
     id=#1
     possessed_class=Overlay
     tracks [
      {
       type=MovieSceneFloatTrack
       display_name=Render Opacity
       sections [{type=MovieSceneFloatSection, end_frame=16, start_frame=0, channels [None {type=MovieSceneScriptingFloatChannel, keys [[0,0],[8,1]]}]}]
      }
      {
       type=MovieScene2DTransformTrack
       sections [{
         type=MovieScene2DTransformSection
         end_frame=13
         start_frame=0
         channels [
          Scale.X {type=MovieSceneScriptingFloatChannel, keys [[0,1.05],[8,1]]}
          Scale.Y {type=MovieSceneScriptingFloatChannel, keys [[0,1.05],[8,1]]}
         ]
        }]
      }
     ]
    }]}
 }
 Anim_Shake {
  binding_count=1
  length_seconds=0.75
  object=&self:Anim_Shake
  sequence {bindings [Overlay_Root {
     id=#2
     display_name=Overlay_Bar
     possessed_class=Overlay
     tracks [{
       type=MovieScene2DTransformTrack
       sections [{
         type=MovieScene2DTransformSection
         end_frame=15
         start_frame=0
         channels [
          Translation.X {type=MovieSceneScriptingFloatChannel, const [0,1]}
          Translation.Y {type=MovieSceneScriptingFloatChannel, keys [[6,0],[9.0003,10],[12.0003,1]]}
          Scale.X {type=MovieSceneScriptingFloatChannel, const [1,1]}
          Scale.Y {type=MovieSceneScriptingFloatChannel, const [1,1]}
         ]
        }]
      }]
    }]}
 }
 ANIM_MaskPhaseIntro {
  binding_count=1
  length_seconds=0.4
  object=&self:ANIM_MaskPhaseIntro
  sequence {bindings [Overlay_MaskPhase {
     id=#3
     possessed_class=Overlay
     tracks [
      {
       type=MovieScene2DTransformTrack
       sections [{
         type=MovieScene2DTransformSection
         end_frame=8
         start_frame=0
         channels [
          Translation.X {type=MovieSceneScriptingFloatChannel, const [0,2]}
          Translation.Y {type=MovieSceneScriptingFloatChannel, keys [[0,-33.6192],[5.7033,4.8329],[8,0],[10,0]]}
          Angle {type=MovieSceneScriptingFloatChannel, const [0,1]}
          Scale.X {type=MovieSceneScriptingFloatChannel, keys [[0,0.8],[5.7033,1],[10,1]]}
          Scale.Y {type=MovieSceneScriptingFloatChannel, keys [[0,0.8],[5.7033,1],[10,1]]}
          Shear.X {type=MovieSceneScriptingFloatChannel, const [0,1]}
          Shear.Y {type=MovieSceneScriptingFloatChannel, const [0,1]}
         ]
        }]
      }
      {
       type=MovieSceneFloatTrack
       display_name=Render Opacity
       sections [{type=MovieSceneFloatSection, end_frame=5, start_frame=0, channels [None {type=MovieSceneScriptingFloatChannel, keys [[0,0],[5,1]]}]}]
      }
     ]
    }]}
 }
 ANIM_MaskPhaseOutro {
  binding_count=1
  length_seconds=0.25
  object=&self:ANIM_MaskPhaseOutro
  sequence {bindings [Overlay_MaskPhase {
     id=#3
     possessed_class=Overlay
     tracks [
      {
       type=MovieScene2DTransformTrack
       sections [{
         type=MovieScene2DTransformSection
         end_frame=5
         start_frame=0
         channels [
          Translation.X {type=MovieSceneScriptingFloatChannel, const [0,1]}
          Translation.Y {type=MovieSceneScriptingFloatChannel, keys [[0,0],[5,-52.1317],[10,0]]}
          Angle {type=MovieSceneScriptingFloatChannel, const [0,1]}
          Scale.X {type=MovieSceneScriptingFloatChannel, keys [[0,1],[5,0.8],[10,1]]}
          Scale.Y {type=MovieSceneScriptingFloatChannel, keys [[0,1],[5,0.8],[10,1]]}
          Shear.X {type=MovieSceneScriptingFloatChannel, const [0,1]}
          Shear.Y {type=MovieSceneScriptingFloatChannel, const [0,1]}
         ]
        }]
      }
      {
       type=MovieSceneFloatTrack
       display_name=Render Opacity
       sections [{type=MovieSceneFloatSection, end_frame=5, start_frame=0, channels [None {type=MovieSceneScriptingFloatChannel, keys [[0,1],[5,0]]}]}]
      }
     ]
    }]}
 }
]
graphs [
 BarAnimInterpolate {nodes [
   BarAnimInterpolate() {is=title K2Node_FunctionEntry}
   > n1 = Set PreviousProgressValue(PreviousProgressValue:float=float - float(A=Get PreviousProgressValue() -> (PreviousProgressValue:float), B=Get InterpCadence() -> (InterpCadence:double))) -> (Output_Get:float)
   > n2 = Branch(Condition="float <= float"(A=@n1, B=Get CurrentProgressValue() -> (CurrentProgressValue:double)) -> (ReturnValue=~))
   > Clear and Invalidate Timer by Handle(Handle=Get Timer_DeltaAnimateProgressUpdate() -> (Timer_DeltaAnimateProgressUpdate:Timer Handle Structure))
   > n3 = GetDynamicMaterial(self=Get Image_ProgressBar() -> (Image_ProgressBar:Image Object Reference))
   > SetScalarParameterValue(self=@n3, ParameterName=Progress, Value=Get CurrentProgressValue())
   > Set PreviousProgressValue(PreviousProgressValue=Get CurrentProgressValue())
   n4 = GetDynamicMaterial(execute=@n2.else, self=Get Image_ProgressBar())
   > SetScalarParameterValue(self=@n4, ParameterName=Progress, Value=Get PreviousProgressValue())
  ]}
 EventGraph {nodes [
   n1 = UpdateMaskState!() -> (OutputDelegate, then, MaskPhaseActive:bool=~) {is=title K2Node_CustomEvent}
   > n2 = Branch(Condition=@n1.MaskPhaseActive)
   > SetVisibility(self=Get Overlay_MaskPhase() -> (Overlay_MaskPhase:Overlay Object Reference), InVisibility=HitTestInvisible)
   > PlayAnimation(InAnimation=Get ANIM_MaskPhaseIntro() -> (ANIM_MaskPhaseIntro:Widget Animation Object Reference))
   n3 = Play Animation with Finished event!(
    execute:Exec=@n2.else
    Widget:User Widget Object Reference=Self-Reference()
    InAnimation:Widget Animation Object Reference=Get ANIM_MaskPhaseOutro() -> (ANIM_MaskPhaseOutro:Widget Animation Object Reference)
    StartAtTime:float=0
    NumLoopsToPlay:int=1
    PlayMode:EUMGSequencePlayMode Enum
    PlaybackSpeed:float=1
   ) -> (then:Exec, Result:Widget Animation Handle Structure, Finished:Exec) {is=title K2Node_PlayAnimation2}
   SetVisibility(execute=@n3.Finished, self=Get Overlay_MaskPhase(), InVisibility=Collapsed)
   n4 = OnBarAnimationFinished() {is=title K2Node_CustomEvent}
   > Clear and Invalidate Timer by Handle(Handle=Get BarAnimTimer() -> (BarAnimTimer:Timer Handle Structure))
   Event Construct()
   > Parent: Construct()
   > SetProgressDirectly!(execute, self:WBP Challenge Progress Bar Object Reference, Progress:double=1) -> (then)
   > UpdateMaskState!(execute, self, MaskPhaseActive:bool=true) -> (then) {is=title K2Node_CallFunction}
   > InitializeMaskWeakpointIcons()
   n5 = Event SetProgressBarStyle!() -> (OutputDelegate, then, HasReachedBossStage:bool) {is=title K2Node_Event}
   > n6 = Branch(Condition=@n5.HasReachedBossStage)
   > PlayAnimation(InAnimation=Get Anim_SwitchToBossBar() -> (Anim_SwitchToBossBar:Widget Animation Object Reference))
   > Set Timer by Event(Delegate=@n4, Time=1, bLooping=false)
   SetupBossBarElements!(execute=@n6.else, self:WBP Challenge Progress Bar Object Reference, IsInBossStage:bool=false) -> (then)
   Event CallBarAnimationFinishedEvent() {is=title K2Node_Event}
   > Parent: CallBarAnimationFinishedEvent() {is=title K2Node_CallParentFunction}
   n7 = Event PlayShakeAnimation!() -> (OutputDelegate, then, AnimationDelay:double) {is=title K2Node_Event}
   > Delay(Duration=@n7.AnimationDelay)
   > PlayAnimation(InAnimation=Get Anim_Shake() -> (Anim_Shake:Widget Animation Object Reference))
   Event WaitForBarAnimationToFinish() {is=title K2Node_Event}
  ]}
 GetSizeBoxWidth {nodes [
   GetSizeBoxWidth() {is=title K2Node_FunctionEntry}
   > Return Node!(
    execute
    Width:double=Get WidthOverride!(self:Size Box Object Reference=Get SizeBox_ProgressBar() -> (SizeBox_ProgressBar:Size Box Object Reference)) -> (WidthOverride:float)
   )
  ]}
 InitializeMaskWeakpointIcons {nodes [
   InitializeMaskWeakpointIcons() {is=title K2Node_FunctionEntry}
   > Set MaskWeakpoints(MaskWeakpoints:Array of User Widget Object References=Make Array!(
     "[0]":User Widget Object Reference=Get WBP_MaskMaker_Challenge_MaskPip() -> (WBP_MaskMaker_Challenge_MaskPip:WBP Mask Maker Challenge Mask Pip Object Reference)
     "[1]":User Widget Object Reference=Get WBP_MaskMaker_Challenge_MaskPip_1() -> (WBP_MaskMaker_Challenge_MaskPip_1:WBP Mask Maker Challenge Mask Pip Object Reference)
     "[2]":User Widget Object Reference=Get WBP_MaskMaker_Challenge_MaskPip_2() -> (WBP_MaskMaker_Challenge_MaskPip_2:WBP Mask Maker Challenge Mask Pip Object Reference)
     "[3]":User Widget Object Reference=Get WBP_MaskMaker_Challenge_MaskPip_3() -> (WBP_MaskMaker_Challenge_MaskPip_3:WBP Mask Maker Challenge Mask Pip Object Reference)
    ) -> (Array:Array of User Widget Object References)) -> (Output_Get:Array of User Widget Object References)
  ]}
 OnMaskWeakpointHealthUpdated {nodes [
   n1 = OnMaskWeakpointHealthUpdated!() -> (then, WeakpointIndex:int, IsWeakpointActive:bool=~) {is=title K2Node_FunctionEntry}
   n2 = Get Active Weakpoints() -> (Active Weakpoints:Array of Integers)
   > n3 = Branch(Condition=Is Valid Index!(
     TargetArray:Array of Integers=@n2
     IndexToTest:int=Find Item!(TargetArray:Array of Integers=@n2, ItemToFind:"Integer (by ref)"=@n1) -> (ReturnValue)
    ) -> (ReturnValue:bool))
   > n4 = Branch(Condition=Get IsWeakpointActive() -> (IsWeakpointActive:bool))
   n5 = Get Active Weakpoints()
   n6 = Cast To WBP_MaskMaker_Challenge_MaskPip!(
    execute=@n4.else
    Object="Get (a copy)"(
     Array:Array of User Widget Object References=Get MaskWeakpoints() -> (MaskWeakpoints:Array of User Widget Object References)
     Dimension 1=Last Index!(TargetArray:Array of Integers=@n5) -> (ReturnValue=0)
    ) -> (Output:User Widget Object Reference)
   ) -> (then, CastFailed, AsWBP Mask Maker Challenge Mask Pip:WBP Mask Maker Challenge Mask Pip Object Reference) {is=title K2Node_DynamicCast}
   > SetWeakpointActive!(execute, self:WBP Mask Maker Challenge Mask Pip Object Reference=@n6, IsActive:bool=false) -> (then)
   > Remove Item!(execute, TargetArray:Array of Integers=Get Active Weakpoints(), Item:"Integer (by ref)"=Get WeakpointIndex() -> (WeakpointIndex:int=0)) -> (then, ReturnValue:bool)
   Branch(execute=@n3.else, Condition=Get IsWeakpointActive())
   > Add(TargetArray:Array of Integers=Get Active Weakpoints(), NewItem:"Integer (by ref)"=Get WeakpointIndex() -> (WeakpointIndex=0)) {is=title K2Node_CallArrayFunction}
   n7 = Get Active Weakpoints()
   > n8 = Cast To WBP_MaskMaker_Challenge_MaskPip(Object="Get (a copy)"(Array=Get MaskWeakpoints(), Dimension 1=Last Index(TargetArray=@n7) -> (ReturnValue=0)))
   > SetWeakpointActive(self=@n8, IsActive=true)
  ]}
 SetProgressDirectly {nodes [
   n1 = SetProgressDirectly!() -> (then, Progress:double) {is=title K2Node_FunctionEntry}
   > n2 = GetDynamicMaterial(self=Get Image_ProgressBar())
   > SetScalarParameterValue(self=@n2, ParameterName=Progress, Value=@n1)
  ]}
 SetProgressWithAnimation {nodes [
   n1 = SetProgressWithAnimation!() -> (then, ProgressEnd:double) {is=title K2Node_FunctionEntry}
   > Set CurrentProgressValue(CurrentProgressValue:double=@n1) -> (Output_Get:double)
   > n2 = Set Timer by Event(Delegate=Create Event(), Time=0.03, bLooping=true, InitialStartDelay=0.05)
   > Set Timer_DeltaAnimateProgressUpdate(Timer_DeltaAnimateProgressUpdate:Timer Handle Structure=@n2) -> (Output_Get:Timer Handle Structure)
  ]}
 SetWidthMobile {nodes [
   SetWidthMobile() {is=title K2Node_FunctionEntry}
   n1 = GetOwningPlayer()
   > Branch(Condition=AND Boolean(A=IsUsingTouch(PlayerController=@n1), B=IsMobileGame()))
   n2 = GetViewportSize!() -> (ReturnValue_X:double, ReturnValue_Y:double)
   n3 = float / float(A=@n2, B=@n2.ReturnValue_Y)
   > SetWidthOverride(
    self=Get SizeBox_ProgressBar()
    InWidthOverride=Select(Option 0:float=650, Option 1:float=400, Index="float <= float"(A=@n3, B=1.5) -> (ReturnValue=~)) -> (ReturnValue:float)
   )
  ]}
 SetupBossBarElements {nodes [SetupBossBarElements!() -> (then, IsInBossStage:bool) {is=title K2Node_FunctionEntry}]}
 SetupBossIcon {nodes [SetupBossIcon!() -> (then, BossIcon:Texture Object Reference) {is=title K2Node_FunctionEntry}]}
]
variables [
 Active Weakpoints {type="array<integer>"}
 CurrentProgressValue {type=number}
 MaskWeakpoints {type="array<UserWidget>", default []}
 PreviousProgressValue {type=number}
 Timer_DeltaAnimateProgressUpdate {type=TimerHandle}
]
widgets {
 info {
  inheritedWidgetCount=0
  parentClass=/AIProgressBar/UI/WBPs/WBP_Challenge_ProgressBar.WBP_Challenge_ProgressBar_C
  rootWidgetClass=&Overlay
  widgetCount=15
 }
 widgets [
  Overlay_Root {widgetClassPath=&Overlay, widget=Overlay_Root}
  VerticalBox_0 {
   widgetClassPath=&VerticalBox
   parent=Overlay_Root
   widget=VerticalBox_0
   slot {class=OverlaySlot, properties {verticalAlignment=VAlign_Center}}
  }
  Overlay_ProgressBar {widgetClassPath=&Overlay, bIsVariable=true, parent=VerticalBox_0, widget=Overlay_ProgressBar, slot {class=VerticalBoxSlot}}
  SizeBox_ProgressBar {
   widgetClassPath=&SizeBox
   bIsVariable=true
   parent=Overlay_ProgressBar
   widget=SizeBox_ProgressBar
   properties {bOverride_WidthOverride=true, heightOverride=20, widthOverride=800}
   slot {class=OverlaySlot, properties {horizontalAlignment=HAlign_Fill, verticalAlignment=VAlign_Center}}
  }
  Overlay_70 {widgetClassPath=&Overlay, parent=SizeBox_ProgressBar, widget=Overlay_70, slot {class=SizeBoxSlot}}
  Image_ProgressBar_SmokeBG {
   widgetClassPath=&Image
   bIsVariable=true
   parent=Overlay_70
   widget=Image_ProgressBar_SmokeBG
   properties {
    renderOpacity=0.5
    brush {
     bIsDynamicallyLoaded=false
     drawAs=Image
     imageType=FullColor
     mirroring=NoMirror
     resourceName=None
     resourceObject=/MaskMakerBoss/UI/Material/MI_UI_MaskMaker_Smoke
     tiling=NoTile
     imageSize=<32 32>
     margin=<ltrb 0 0 0 0>
     outlineSettings {
      bUseBrushTransparency=false
      roundingType=HalfHeightRadius
      width=0
      color {colorUseRule=UseColor_Specified, specifiedColor=<rgba 0 0 0 0>}
      cornerRadii=<0 0 0 0>
     }
     tintColor {colorUseRule=UseColor_Specified, specifiedColor=<rgba 1 1 1 1>}
     uVRegion {bIsValid=false, max=<0 0>, min=<0 0>}
    }
   }
   slot {class=OverlaySlot, properties {horizontalAlignment=HAlign_Fill, verticalAlignment=VAlign_Fill, padding=<ltrb -20 -10 -20 -10>}}
  }
  Image_ProgressBar_BG {
   widgetClassPath=&Image
   parent=Overlay_70
   widget=Image_ProgressBar_BG
   properties {
    renderOpacity=0.5
    brush {
     bIsDynamicallyLoaded=false
     drawAs=Image
     imageType=FullColor
     mirroring=NoMirror
     resourceName=None
     resourceObject=/MaskMakerBoss/UI/Material/MI_UI_MaskMaker_BarBG
     tiling=NoTile
     imageSize=<32 32>
     margin=<ltrb 0 0 0 0>
     outlineSettings {
      bUseBrushTransparency=false
      roundingType=HalfHeightRadius
      width=0
      color {colorUseRule=UseColor_Specified, specifiedColor=<rgba 0 0 0 0>}
      cornerRadii=<0 0 0 0>
     }
     tintColor {colorUseRule=UseColor_Specified, specifiedColor=<rgba 1 1 1 1>}
     uVRegion {bIsValid=false, max=<0 0>, min=<0 0>}
    }
   }
   slot {class=OverlaySlot, properties {horizontalAlignment=HAlign_Fill, verticalAlignment=VAlign_Fill}}
  }
  Image_ProgressBar {
   widgetClassPath=&Image
   bIsVariable=true
   parent=Overlay_70
   widget=Image_ProgressBar
   properties {brush {
     bIsDynamicallyLoaded=false
     drawAs=Image
     imageType=FullColor
     mirroring=NoMirror
     resourceName=None
     resourceObject=/MaskMakerBoss/UI/Material/MI_UI_MaskMaker_BarFill
     tiling=NoTile
     imageSize=<32 32>
     margin=<ltrb 0 0 0 0>
     outlineSettings {
      bUseBrushTransparency=false
      roundingType=HalfHeightRadius
      width=0
      color {colorUseRule=UseColor_Specified, specifiedColor=<rgba 0 0 0 0>}
      cornerRadii=<0 0 0 0>
     }
     tintColor {colorUseRule=UseColor_Specified, specifiedColor=<rgba 1 1 1 1>}
     uVRegion {bIsValid=false, max=<0 0>, min=<0 0>}
    }}
   slot {class=OverlaySlot, properties {horizontalAlignment=HAlign_Fill, verticalAlignment=VAlign_Fill}}
  }
  Overlay_MaskPhase {
   widgetClassPath=&Overlay
   bIsVariable=true
   parent=VerticalBox_0
   widget=Overlay_MaskPhase
   slot {class=VerticalBoxSlot, properties {horizontalAlignment=HAlign_Center, verticalAlignment=VAlign_Center, padding=<ltrb 0 5 0 0>}}
  }
  Image_Weakpoint_BG {
   widgetClassPath=&Image
   parent=Overlay_MaskPhase
   widget=Image_Weakpoint_BG
   properties {
    renderOpacity=0
    visibility=SelfHitTestInvisible
    brush {
     bIsDynamicallyLoaded=false
     drawAs=Image
     imageType=FullColor
     mirroring=NoMirror
     resourceName=None
     resourceObject=/MaskMakerBoss/UI/Material/MI_UI_MaskMaker_bg
     tiling=NoTile
     imageSize=<32 32>
     margin=<ltrb 0 0 0 0>
     outlineSettings {
      bUseBrushTransparency=false
      roundingType=HalfHeightRadius
      width=0
      color {colorUseRule=UseColor_Specified, specifiedColor=<rgba 0 0 0 0>}
      cornerRadii=<0 0 0 0>
     }
     tintColor {colorUseRule=UseColor_Specified, specifiedColor=<rgba 1 1 1 0.75>}
     uVRegion {bIsValid=false, max=<0 0>, min=<0 0>}
    }
   }
   slot {class=OverlaySlot, properties {horizontalAlignment=HAlign_Fill, verticalAlignment=VAlign_Fill, padding=<ltrb -10 0 -10 0>}}
  }
  HBox_MaskWeakpoints {
   widgetClassPath=&HorizontalBox
   parent=Overlay_MaskPhase
   widget=HBox_MaskWeakpoints
   slot {class=OverlaySlot, properties {horizontalAlignment=HAlign_Center, verticalAlignment=VAlign_Center, padding=<ltrb 2 2 2 2>}}
  }
  WBP_MaskMaker_Challenge_MaskPip {
   widgetClassPath=&/WBP_MaskMaker_Challenge_MaskPip
   bIsVariable=true
   parent=HBox_MaskWeakpoints
   widget=WBP_MaskMaker_Challenge_MaskPip
   slot {class=HorizontalBoxSlot, properties {padding=<ltrb 3 3 3 3>}}
  }
  WBP_MaskMaker_Challenge_MaskPip_1 {
   widgetClassPath=&WBP_MaskMaker_Challenge_MaskPip
   bIsVariable=true
   parent=HBox_MaskWeakpoints
   widget=WBP_MaskMaker_Challenge_MaskPip_1
   slot {class=HorizontalBoxSlot, properties {padding=<ltrb 3 3 3 3>}}
  }
  WBP_MaskMaker_Challenge_MaskPip_2 {
   widgetClassPath=&WBP_MaskMaker_Challenge_MaskPip
   bIsVariable=true
   parent=HBox_MaskWeakpoints
   widget=WBP_MaskMaker_Challenge_MaskPip_2
   slot {class=HorizontalBoxSlot, properties {padding=<ltrb 3 3 3 3>}}
  }
  WBP_MaskMaker_Challenge_MaskPip_3 {
   widgetClassPath=&WBP_MaskMaker_Challenge_MaskPip
   bIsVariable=true
   parent=HBox_MaskWeakpoints
   widget=WBP_MaskMaker_Challenge_MaskPip_3
   slot {class=HorizontalBoxSlot, properties {padding=<ltrb 3 3 3 3>}}
  }
 ]
}
