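# Challenge HUD widget for the Mask Maker boss fight, derived from WBP_Challenge_WidgetBase, showing the boss portrait, name, progress bar and phase hint text.
# Setup casts to ChallengeManager_MaskMaker and binds mask phase and weakpoint health events that update the bar's mask state.
# A scalable float from MaskMaker_GameData gates the hint switching between attacking the masks and the boss.
# Overrides drive intro/outro animations, bar shakes and style changes, and collapse the widget after completion.
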WidgetBlueprint /MaskMakerBoss/UI/WBP_MaskMaker_Challenge_Widget
animations [
 Anim_Intro {
  binding_count=1
  length_seconds=4.2
  object=&self:Anim_Intro
  sequence {
   bindings [Horizontal_BossName {
     id=#1
     possessed_class=HorizontalBox
     tracks [
      {
       type=MovieSceneFloatTrack
       display_name=Render Opacity
       sections [{
         type=MovieSceneFloatSection
         end_frame=252
         start_frame=0
         channels [None {type=MovieSceneScriptingFloatChannel, keys [[0,0],[45,1],[234,1],[252,0]]}]
        }]
      }
      {
       type=MovieScene2DTransformTrack
       sections [{
         type=MovieScene2DTransformSection
         end_frame=240
         start_frame=0
         channels [
          Translation.X {type=MovieSceneScriptingFloatChannel, const [0,2]}
          Translation.Y {type=MovieSceneScriptingFloatChannel, keys [[0,42.1397],[45,0],[234,0]]}
          Angle {type=MovieSceneScriptingFloatChannel, const [0,2]}
          Scale.X {type=MovieSceneScriptingFloatChannel, const [1,2]}
          Scale.Y {type=MovieSceneScriptingFloatChannel, const [1,2]}
          Shear.X {type=MovieSceneScriptingFloatChannel, const [0,2]}
          Shear.Y {type=MovieSceneScriptingFloatChannel, const [0,2]}
         ]
        }]
      }
     ]
    }]
   display_rate {numerator=60}
  }
 }
 Anim_Outro {
  binding_count=1
  length_seconds=5
  object=&self:Anim_Outro
  sequence {bindings [WBP_MaskMaker_Challenge_StateIndicator {
     id=#2
     possessed_class=WBP_MaskMaker_Challenge_StateIndicator_C
     tracks [{
       type=MovieSceneFloatTrack
       display_name=Render Opacity
       sections [{type=MovieSceneFloatSection, end_frame=0, start_frame=0, channels [None {type=MovieSceneScriptingFloatChannel, const [0,1]}]}]
      }]
    }]}
 }
]
graphs [
 BeginChallengeInitialSetup {nodes [
   BeginChallengeInitialSetup() {is=title K2Node_FunctionEntry}
   > InitialSetupOfChallengeVariables(self:WBP Challenge Widget Base Object Reference)
   > n1 = Cast To ChallengeManager_MaskMaker!(execute, Object=Get ChallengeManagerRef() -> (ChallengeManagerRef:Challenge Manager Object Reference)) -> (then, CastFailed, AsChallenge Manager Mask Maker:Challenge Manager Mask Maker Object Reference) {is=title K2Node_DynamicCast}
   > Bind Event to On Boss Mask Phase Updated(self:Challenge Manager Mask Maker Object Reference=@n1, Delegate=Create Event()) {is=title K2Node_AddDelegate}
   > Bind Event to On Weakpoints Health Updated(self:Challenge Manager Mask Maker Object Reference=@n1, Delegate=Create Event()) {is=title K2Node_AddDelegate}
   > InitialSetupOfTextsAndStageMarkers!(
    execute
    self:WBP Challenge Widget Base Object Reference
    Challenge_StageIndicatorText:WBP Challenge Stage Indicator Text Object Reference=Get WBP_MaskMaker_Challenge_StateIndicator() -> (WBP_MaskMaker_Challenge_StateIndicator:WBP Mask Maker Challenge State Indicator Object Reference)
    Challenge Stage Markers Container:WBP Challenge Stage Markers Container Object Reference
    Challenge Progress Bar:WBP Challenge Progress Bar Object Reference=Get WBP_MaskMaker_Challenge_ProgressBar() -> (WBP_MaskMaker_Challenge_ProgressBar:WBP Mask Maker Challenge Progress Bar Object Reference)
   ) -> (then)
   > SetVisibility(InVisibility=HitTestInvisible)
   > ForceUpdateProgressBarVisibility!(execute, self:WBP Challenge Widget Base Object Reference, Visible:ESlateVisibility Enum=HitTestInvisible) -> (then)
   > PlayIntroAnimation(self:WBP Challenge Widget Base Object Reference)
  ]}
 EventGraph {nodes [
   n1 = OnBossMaskPhaseUpdated!() -> (OutputDelegate, then, MaskPhaseActive:bool=~) {is=title K2Node_CustomEvent}
   > n2 = Set Mask Phase Active(Mask Phase Active:bool=@n1.MaskPhaseActive) -> (Output_Get:bool)
   > UpdateMaskState!(execute, self:WBP Mask Maker Challenge Progress Bar Object Reference=Get WBP_MaskMaker_Challenge_ProgressBar(), MaskPhaseActive:bool=@n2) -> (then)
   > Branch(Condition=GetValueAsBool!(Input:"Scalable Float Structure (by ref)"=Get SF_ShouldShowPhaseText() -> (SF_ShouldShowPhaseText:Scalable Float Structure), Level:float=0) -> (ReturnValue:bool))
   n3 = Get WBP_MaskMaker_Challenge_StateIndicator()
   > SetStageText!(
    execute
    self:WBP Challenge Stage Indicator Text Object Reference=@n3
    StageText:Text=SelectText(A=Get MasksActive_StageText() -> (MasksActive_StageText:Text), B=Get MasksInactive_StageText() -> (MasksInactive_StageText:Text), bPickA=@n2)
   ) -> (then)
   > PlayIntroAnimations!(execute, self:WBP Challenge Stage Indicator Text Object Reference=@n3, WithDelayedOutroAnim:bool=true) -> (then)
   Event CheckBossReadyForFight() {is=title K2Node_Event}
   n4 = Get ChallengeManagerRef()
   > Is Valid(InputObject=@n4)
   > n5 = Branch(Condition=Get BossIsReadyForFinalFight!(self:Challenge Manager Object Reference=@n4) -> (BossIsReadyForFinalFight:bool))
   > UpdateChallengeInformation!(execute, self:WBP Challenge Widget Base Object Reference, WithAnimation:bool=false) -> (then)
   > HandleBossProgressUpdatedInternal!(execute, self:WBP Challenge Widget Base Object Reference, AnimateProgressBar:bool=false) -> (then)
   UpdateStageMarkersState(execute=@n5.else, self:WBP Challenge Widget Base Object Reference)
   > PlayShakeAnimation!(execute, self:WBP Challenge Progress Bar Object Reference=Get WBP_MaskMaker_Challenge_ProgressBar(), AnimationDelay:double=0) -> (then)
   Event Construct()
   > Parent: Construct()
   Event OnChallengeCompletedVisualUpdates() {is=title K2Node_Event}
   > Delay(Duration=Get TimeToHideWidgetOnceFinished() -> (TimeToHideWidgetOnceFinished:double))
   > n6 = Play Animation with Finished event!(
    execute:Exec
    Widget:User Widget Object Reference=Self-Reference()
    InAnimation:Widget Animation Object Reference=Get Anim_Outro() -> (Anim_Outro:Widget Animation Object Reference)
    StartAtTime:float=0
    NumLoopsToPlay:int=1
    PlayMode:EUMGSequencePlayMode Enum
    PlaybackSpeed:float=1
   ) -> (then:Exec, Result:Widget Animation Handle Structure, Finished:Exec) {is=title K2Node_PlayAnimation2}
   SetVisibility(execute=@n6.Finished, InVisibility=Collapsed)
   Event PlayIntroAnimation() {is=title K2Node_Event}
   > Play Animation with Finished event(
    Widget=Self-Reference()
    InAnimation=Get Anim_Intro() -> (Anim_Intro:Widget Animation Object Reference)
    StartAtTime=0
    NumLoopsToPlay=1
    PlaybackSpeed=1
   )
   > n7 = Branch(Condition=GetValueAsBool(Input=Get SF_ShouldShowPhaseText(), Level=0))
   > Delay(Duration=0.5)
   > SetVisibility(self=Get WBP_MaskMaker_Challenge_StateIndicator(), InVisibility=HitTestInvisible)
   > OnBossMaskPhaseUpdated!(execute, self, MaskPhaseActive:bool=Get Mask Phase Active() -> (Mask Phase Active:bool)) -> (then) {is=title K2Node_CallFunction}
   SetVisibility(execute=@n7.else, self=Get WBP_MaskMaker_Challenge_StateIndicator(), InVisibility=Collapsed)
   Event PlayOutroAnimation() {is=title K2Node_Event}
   > n8 = Play Animation with Finished event(Widget=Self-Reference(), InAnimation=Get Anim_Outro(), StartAtTime=0, NumLoopsToPlay=1, PlaybackSpeed=1)
   SetupBossBarElements!(execute=@n8.Finished, self:WBP Challenge Progress Bar Object Reference=Get WBP_MaskMaker_Challenge_ProgressBar(), IsInBossStage:bool=true) -> (then)
  ]}
 ForceUpdateProgressBarVisibility {nodes [
   n1 = ForceUpdateProgressBarVisibility!() -> (then, Visible:ESlateVisibility Enum=Visible) {is=title K2Node_FunctionEntry}
   > Parent: ForceUpdateProgressBarVisibility!(execute, Visible:ESlateVisibility Enum=@n1) -> (then) {is=title K2Node_CallParentFunction}
   > SetVisibility(self=Get WBP_MaskMaker_Challenge_ProgressBar(), InVisibility=@n1)
   > Branch(Condition=GetValueAsBool(Input=Get SF_ShouldShowPhaseText(), Level=0))
   > SetVisibility(self=Get WBP_MaskMaker_Challenge_StateIndicator(), InVisibility=@n1)
  ]}
 OnMaskWeakpointsHealthUpdated {nodes [
   n1 = OnMaskWeakpointsHealthUpdated!() -> (then, WeakpointIndex:int, WeakpointHealthPercent:double=~) {is=title K2Node_FunctionEntry}
   n2 = NOT Boolean(A="Nearly Equal (Float)"(A=@n1.WeakpointHealthPercent, B=0, ErrorTolerance=0))
   > OnMaskWeakpointHealthUpdated!(
    execute
    self:WBP Mask Maker Challenge Progress Bar Object Reference=Get WBP_MaskMaker_Challenge_ProgressBar()
    WeakpointIndex:int=@n1
    IsWeakpointActive:bool=@n2
   ) -> (then)
   > Set Mask Phase Active(Mask Phase Active=@n2)
  ]}
 PhaseCompletedVisualUpdates {nodes [
   PhaseCompletedVisualUpdates() {is=title K2Node_FunctionEntry}
   > Unbind Event from On Bar Fill Animation Finished(self:WBP Challenge Progress Bar Object Reference=Get WBP_MaskMaker_Challenge_ProgressBar(), Delegate=Create Event()) {is=title K2Node_RemoveDelegate}
   > UpdateStageMarkersState()
   > PlayShakeAnimation(self=Get WBP_MaskMaker_Challenge_ProgressBar(), AnimationDelay=0)
   n1 = IsLastPhase!(self:Challenge Manager Object Reference=Get ChallengeManagerRef(), PhaseIndex:int=Get CurrentPhase() -> (CurrentPhase:int=0)) -> (Value:bool)
   > n2 = Branch(Condition=@n1)
   > Set Timer by Event(Delegate=Create Event(), Time=4.25, bLooping=false)
   SetProgressBarStyle!(execute=@n2.else, self:WBP Challenge Progress Bar Object Reference=Get WBP_MaskMaker_Challenge_ProgressBar(), HasReachedBossStage:bool=@n1) -> (then)
  ]}
 UpdateChallengeCompleted {nodes [
   n1 = UpdateChallengeCompleted!() -> (then, IsChallengeCompleted:bool) {is=title K2Node_FunctionEntry}
   > n2 = Branch(Condition=@n1)
   > OnChallengeCompletedVisualUpdates(self:WBP Challenge Widget Base Object Reference)
   > SetProgressWithAnimation!(execute, self:WBP Challenge Progress Bar Object Reference=Get WBP_MaskMaker_Challenge_ProgressBar(), ProgressEnd:double=0) -> (then)
   n3 = Get WBP_MaskMaker_Challenge_ProgressBar()
   WaitForBarAnimationToFinish(execute=@n2.else, self:WBP Challenge Progress Bar Object Reference=@n3)
   > Bind Event to On Bar Fill Animation Finished(self:WBP Challenge Progress Bar Object Reference=@n3, Delegate=Create Event()) {is=title K2Node_AddDelegate}
  ]}
 UpdateChallengeInformation {nodes [
   UpdateChallengeInformation!() -> (then, WithAnimation:bool) {is=title K2Node_FunctionEntry}
   > SetProgressBarStyle(self=Get WBP_MaskMaker_Challenge_ProgressBar(), HasReachedBossStage=IsOnBossStage!() -> (IsOnBossStage:bool))
   > PlayIntroAnimation()
  ]}
 UpdateProgressBarFill {nodes [
   UpdateProgressBarFill!() -> (then, With Animation:bool) {is=title K2Node_FunctionEntry}
   > SetProgressDirectly!(
    execute
    self:WBP Challenge Progress Bar Object Reference=Get WBP_MaskMaker_Challenge_ProgressBar()
    Progress:double=Get Boss Progress() -> (Boss Progress:double)
   ) -> (then)
  ]}
 UpdateStageMarkersState {nodes [UpdateStageMarkersState() {is=title K2Node_FunctionEntry}]}
]
variables [
 Mask Phase Active {type=boolean, default=true}
 MasksActive_StageText {type=string, default="Attack the masks!"}
 MasksInactive_StageText {type=string, default="Attack the Mask Maker! "}
 SF_ShouldShowPhaseText {
  type=ScalableFloat
  default {registryType=None, value=1, curve {curveTable=/MaskMakerBoss/DataTable/MaskMaker_GameData, rowName=MaskMaker.UI.ShowPhaseText}}
 }
]
widgets {
 info {
  inheritedWidgetCount=1
  parentClass=/AIProgressBar/UI/WBPs/WBP_Challenge_WidgetBase.WBP_Challenge_WidgetBase_C
  rootWidgetClass=&Overlay
  widgetCount=10
 }
 widgets [
  Overlay_Root {widgetClassPath=&Overlay, widget=Overlay_Root}
  VerticalBox_26 {
   widgetClassPath=&VerticalBox
   parent=Overlay_Root
   widget=VerticalBox_26
   slot {class=OverlaySlot, properties {horizontalAlignment=HAlign_Fill, verticalAlignment=VAlign_Center}}
  }
  Horizontal_BossName {
   widgetClassPath=&HorizontalBox
   parent=VerticalBox_26
   widget=Horizontal_BossName
   properties {bOverride_Cursor=true}
   slot {class=VerticalBoxSlot, properties {horizontalAlignment=HAlign_Center, verticalAlignment=VAlign_Center, padding=<ltrb 0 0 0 4>}}
  }
  Overlay_CharPortrait {
   widgetClassPath=&Overlay
   parent=Horizontal_BossName
   widget=Overlay_CharPortrait
   slot {class=HorizontalBoxSlot, properties {padding=<ltrb 0 -15 15 0>}}
  }
  Image_CharPortraitBG {
   widgetClassPath=&Image
   bIsVariable=true
   parent=Overlay_CharPortrait
   widget=Image_CharPortraitBG
   properties {
    visibility=HitTestInvisible
    brush {
     bIsDynamicallyLoaded=false
     drawAs=Image
     imageType=FullColor
     mirroring=NoMirror
     resourceName=None
     resourceObject=/MaskMakerBoss/UI/Material/M_UI_MaskMaker_MaskPhase
     tiling=NoTile
     imageSize=<64 64>
     margin=<ltrb 0 0 0 0>
     outlineSettings {
      bUseBrushTransparency=false
      roundingType=HalfHeightRadius
      width=0
      color {colorUseRule=UseColor_Specified, specifiedColor=<rgba 0 0 0 0>}
      cornerRadii=<0 0 0 0>
     }
     tintColor {colorUseRule=UseColor_Specified, specifiedColor=<rgba 1 1 1 1>}
     uVRegion {bIsValid=false, max=<0 0>, min=<0 0>}
    }
   }
   slot {class=OverlaySlot}
  }
  Image_CharPortraitAvatar {
   widgetClassPath=&Image
   bIsVariable=true
   parent=Overlay_CharPortrait
   widget=Image_CharPortraitAvatar
   properties {
    visibility=HitTestInvisible
    brush {
     bIsDynamicallyLoaded=false
     drawAs=Image
     imageType=FullColor
     mirroring=NoMirror
     resourceName=None
     resourceObject=/MaskMakerBoss/UI/Textures/T_UI_Icon_MaskMakerPortrait
     tiling=NoTile
     imageSize=<50 50>
     margin=<ltrb 0 0 0 0>
     outlineSettings {
      bUseBrushTransparency=false
      roundingType=HalfHeightRadius
      width=0
      color {colorUseRule=UseColor_Specified, specifiedColor=<rgba 0 0 0 0>}
      cornerRadii=<0 0 0 0>
     }
     tintColor {colorUseRule=UseColor_Specified, specifiedColor=<rgba 1 1 1 1>}
     uVRegion {bIsValid=false, max=<0 0>, min=<0 0>}
    }
   }
   slot {class=OverlaySlot, properties {horizontalAlignment=HAlign_Center, verticalAlignment=VAlign_Center, padding=<ltrb 0 -5 0 0>}}
  }
  Text_BossName {
   widgetClassPath=&CommonTextBlock
   parent=Horizontal_BossName
   widget=Text_BossName
   properties {
    justification=Center
    text=Daigo
    textTransformPolicy=ToUpper
    font {
     bForceMonospaced=false
     bMaterialIsStencil=false
     fontMaterial=null
     fontObject=&Burbank
     letterSpacing=36
     monospacedWidth=0.552
     size=24
     skewAmount=0
     typefaceFontName=Burbank-Black
     outlineSettings {
      bApplyOutlineToDropShadows=false
      bMiteredCorners=false
      bSeparateFillAlpha=false
      outlineMaterial=null
      outlineSize=0
      outlineColor=<rgba 0.0781 0.0781 0.0781 0.235>
     }
    }
    shadowColorAndOpacity=<rgba 0 0 0 0.56>
    shadowOffset=<1.5 1.5>
   }
   slot {class=HorizontalBoxSlot}
  }
  WBP_MaskMaker_Challenge_ProgressBar {
   widgetClassPath=&/WBP_MaskMaker_Challenge_ProgressBar
   bIsVariable=true
   parent=VerticalBox_26
   widget=WBP_MaskMaker_Challenge_ProgressBar
   properties {bOverride_Cursor=true}
   slot {class=VerticalBoxSlot, properties {horizontalAlignment=HAlign_Center}}
  }
  WBP_MaskMaker_Challenge_StateIndicator {
   widgetClassPath=&/WBP_MaskMaker_Challenge_StateIndicator
   bIsVariable=true
   parent=VerticalBox_26
   widget=WBP_MaskMaker_Challenge_StateIndicator
   properties {visibility=Collapsed}
   slot {class=VerticalBoxSlot, properties {horizontalAlignment=HAlign_Center, verticalAlignment=VAlign_Center}}
  }
  TutorialHighlight {widgetClassPath=&TutorialHighlight}
 ]
}
